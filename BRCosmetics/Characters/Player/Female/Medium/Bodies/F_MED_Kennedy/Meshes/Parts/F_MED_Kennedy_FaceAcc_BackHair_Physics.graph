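# Physics asset for the back hair of the female medium Kennedy character, built from capsule bodies.
# Center, left and right four-bone chains hang from base bones on the head, with extra branch bodies on the left and right.
# Hair joints have zero twist limits and narrow swing limits that mostly widen toward the tips.
# Head, neck, upper spine, clavicle and upper-arm capsules form the surrounding body, linked by free-rotation constraints.
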PhysicsAsset /BRCosmetics/Characters/Player/Female/Medium/Bodies/F_MED_Kennedy/Meshes/Parts/F_MED_Kennedy_FaceAcc_BackHair_Physics
bodies {
 base_Knd_C_hair_back [{shapeType=Capsule, length=0.505, radius=0.505, shapeName=base_Knd_C_hair_back_capsule, rotation=<rot 0 -90 0>}]
 base_Knd_L_hair_back [{shapeType=Capsule, length=0.505, radius=0.505, shapeName=base_Knd_L_hair_back_capsule, rotation=<rot 0 -90 0>}]
 base_Knd_R_hair_back [{shapeType=Capsule, length=0.505, radius=0.505, shapeName=base_Knd_R_hair_back_capsule, rotation=<rot 0 -90 0>}]
 clavicle_l [{shapeType=Capsule, length=6.3, radius=8, shapeName=clavicle_l_capsule, center=<7.2 -1.2 -3.6>, rotation=<rot 90 0 0>}]
 clavicle_r [{shapeType=Capsule, length=6.3, radius=8, shapeName=clavicle_l_capsule, center=<-7.2 -1.2 3.6>, rotation=<rot 90 0 0>}]
 dyn_Knd_C_hair_back_1 [{shapeType=Capsule, length=2, radius=2, shapeName=dyn_Knd_C_hair_back_1_capsule, center=<2 0 0>, rotation=<rot 90 0 0>}]
 dyn_Knd_C_hair_back_2 [{shapeType=Capsule, length=1.5455, radius=2.4545, shapeName=dyn_Knd_C_hair_back_2_capsule, center=<2 0 0>, rotation=<rot 90 0 0>}]
 dyn_Knd_C_hair_back_3 [{shapeType=Capsule, length=2.5516, radius=3.0991, shapeName=dyn_Knd_C_hair_back_3_capsule, center=<2.7149 0 0>, rotation=<rot 90 0 0>}]
 dyn_Knd_C_hair_back_4 [{shapeType=Capsule, length=5.46, radius=3.6, shapeName=dyn_Knd_C_hair_back_4_capsule, center=<4.3042 0.3 0>, rotation=<rot 0 0.0004 0>}]
 dyn_Knd_L_hair_back_1 [{shapeType=Capsule, length=2, radius=2, shapeName=dyn_Knd_L_hair_back_1_capsule, center=<2 0 0>, rotation=<rot 90 0 0>}]
 dyn_Knd_L_hair_back_2 [{shapeType=Capsule, length=4, radius=3.5, shapeName=dyn_Knd_L_hair_back_2_capsule, center=<3 -1 0>, rotation=<rot 0 0.0002 20>}]
 dyn_Knd_L_hair_back_3 [{shapeType=Capsule, length=5, radius=4, shapeName=dyn_Knd_L_hair_back_3_capsule, center=<5 -1 0>, rotation=<rot 0 0.0002 10>}]
 dyn_Knd_L_hair_back_3_a [{shapeType=Capsule, length=4.1, radius=4.45, shapeName=dyn_Knd_L_hair_back_3_a_capsule, center=<2.5 0 0>, rotation=<rot 0 0 90>}]
 dyn_Knd_L_hair_back_4 [{shapeType=Capsule, length=5.5, radius=3.5, shapeName=dyn_Knd_L_hair_back_4_capsule, center=<4 -1 0>, rotation=<rot 10 0.0004 0>}]
 dyn_Knd_R_hair_back_1 [{shapeType=Capsule, length=2, radius=2, shapeName=dyn_Knd_R_hair_back_1_capsule, center=<2 0 0>, rotation=<rot 90 0 0>}]
 dyn_Knd_R_hair_back_1_a [{shapeType=Capsule, length=2.7271, radius=3, shapeName=dyn_Knd_R_hair_back_1_a_capsule, center=<3.6075 0 0>}]
 dyn_Knd_R_hair_back_2 [{shapeType=Capsule, length=4, radius=3.5, shapeName=dyn_Knd_R_hair_back_2_capsule, center=<3 -1 0>, rotation=<rot 0 0.0001 -20>}]
 dyn_Knd_R_hair_back_3 [{shapeType=Capsule, length=5, radius=4, shapeName=dyn_Knd_R_hair_back_3_capsule, center=<5 -1 0>, rotation=<rot 0 0.0001 -20>}]
 dyn_Knd_R_hair_back_3_a [{
   shapeType=Capsule
   length=5.5
   radius=3
   shapeName=dyn_Knd_R_hair_back_3_a_capsule
   center=<4 -3 -2.5>
   rotation=<rot 65.1858 169.5838 51.9237>
  }]
 dyn_Knd_R_hair_back_4 [{shapeType=Capsule, length=5.5, radius=3.5, shapeName=dyn_Knd_R_hair_back_4_capsule, center=<4 -1 0>, rotation=<rot -9.8466 1.7539 -10.1511>}]
 head [{shapeType=Capsule, length=3, radius=3, shapeName=head_capsule, rotation=<rot 90 0 0>}]
 neck_01 [{shapeType=Capsule, length=5, radius=3, shapeName=neck_01_capsule, rotation=<rot 90 0 0>}]
 neck_02 [{shapeType=Capsule, length=3, radius=3, shapeName=neck_02_capsule, rotation=<rot 90 0 0>}]
 spine_05 [{shapeType=Capsule, length=13.8625, radius=9.5298, shapeName=spine_05_capsule, center=<1.6247 -0.3707 0>, rotation=<rot 0 -90 0>}]
 upperarm_l [{shapeType=Capsule, length=26.7117, radius=4.9758, shapeName=upperarm_l_capsule, center=<16.7022 0 0>, rotation=<rot 90 0 0>}]
 upperarm_r [{shapeType=Capsule, length=26.7117, radius=4.9758, shapeName=upperarm_l_capsule, center=<-16.7022 0 0>, rotation=<rot 90 0 0>}]
}
constraints [
 {
  bone1Name=neck_01
  bone2Name=spine_05
  swing1LimitDegrees=45
  swing1Motion=Free
  swing2LimitDegrees=45
  swing2Motion=Free
  twistLimitDegrees=45
  twistMotion=Free
 }
 {
  bone1Name=clavicle_l
  bone2Name=spine_05
  swing1LimitDegrees=45
  swing1Motion=Free
  swing2LimitDegrees=45
  swing2Motion=Free
  twistLimitDegrees=45
  twistMotion=Free
 }
 {
  bone1Name=clavicle_r
  bone2Name=spine_05
  swing1LimitDegrees=45
  swing1Motion=Free
  swing2LimitDegrees=45
  swing2Motion=Free
  twistLimitDegrees=45
  twistMotion=Free
 }
 {
  bone1Name=base_Knd_C_hair_back
  bone2Name=head
  swing1LimitDegrees=45
  swing1Motion=Free
  swing2LimitDegrees=45
  swing2Motion=Free
  twistLimitDegrees=45
  twistMotion=Free
 }
 {bone1Name=dyn_Knd_C_hair_back_1, bone2Name=base_Knd_C_hair_back, swing1LimitDegrees=10, swing2LimitDegrees=5, twistLimitDegrees=0}
 {bone1Name=dyn_Knd_C_hair_back_2, bone2Name=dyn_Knd_C_hair_back_1, swing1LimitDegrees=5, swing2LimitDegrees=5, twistLimitDegrees=0}
 {bone1Name=dyn_Knd_C_hair_back_3, bone2Name=dyn_Knd_C_hair_back_2, swing1LimitDegrees=20, swing2LimitDegrees=5, twistLimitDegrees=0}
 {bone1Name=dyn_Knd_C_hair_back_4, bone2Name=dyn_Knd_C_hair_back_3, swing1LimitDegrees=30, swing2LimitDegrees=5, twistLimitDegrees=0}
 {
  bone1Name=base_Knd_L_hair_back
  bone2Name=head
  swing1LimitDegrees=45
  swing1Motion=Free
  swing2LimitDegrees=45
  swing2Motion=Free
  twistLimitDegrees=45
  twistMotion=Free
 }
 {bone1Name=dyn_Knd_L_hair_back_1, bone2Name=base_Knd_L_hair_back, swing1LimitDegrees=5, swing2LimitDegrees=5, twistLimitDegrees=0}
 {bone1Name=dyn_Knd_L_hair_back_2, bone2Name=dyn_Knd_L_hair_back_1, swing1LimitDegrees=5, swing2LimitDegrees=5, twistLimitDegrees=0}
 {bone1Name=dyn_Knd_L_hair_back_3, bone2Name=dyn_Knd_L_hair_back_2, swing1LimitDegrees=20, swing2LimitDegrees=5, twistLimitDegrees=0}
 {bone1Name=dyn_Knd_L_hair_back_4, bone2Name=dyn_Knd_L_hair_back_3, swing1LimitDegrees=30, swing2LimitDegrees=5, twistLimitDegrees=0}
 {
  bone1Name=base_Knd_R_hair_back
  bone2Name=head
  swing1LimitDegrees=45
  swing1Motion=Free
  swing2LimitDegrees=45
  swing2Motion=Free
  twistLimitDegrees=45
  twistMotion=Free
 }
 {bone1Name=dyn_Knd_R_hair_back_1, bone2Name=base_Knd_R_hair_back, swing1LimitDegrees=5, swing2LimitDegrees=5, twistLimitDegrees=0}
 {bone1Name=dyn_Knd_R_hair_back_2, bone2Name=dyn_Knd_R_hair_back_1, swing1LimitDegrees=5, swing2LimitDegrees=5, twistLimitDegrees=0}
 {bone1Name=dyn_Knd_R_hair_back_3, bone2Name=dyn_Knd_R_hair_back_2, swing1LimitDegrees=20, swing2LimitDegrees=5, twistLimitDegrees=0}
 {bone1Name=dyn_Knd_R_hair_back_4, bone2Name=dyn_Knd_R_hair_back_3, swing1LimitDegrees=30, swing2LimitDegrees=5, twistLimitDegrees=0}
 {bone1Name=dyn_Knd_L_hair_back_3_a, bone2Name=dyn_Knd_L_hair_back_3, swing1LimitDegrees=22, swing2LimitDegrees=22, twistLimitDegrees=0}
 {bone1Name=dyn_Knd_R_hair_back_3_a, bone2Name=dyn_Knd_R_hair_back_1_a, swing1LimitDegrees=15, swing2LimitDegrees=15, twistLimitDegrees=0}
 {bone1Name=dyn_Knd_R_hair_back_1_a, bone2Name=dyn_Knd_R_hair_back_1, swing1LimitDegrees=15, swing2LimitDegrees=15, twistLimitDegrees=0}
 {
  bone1Name=upperarm_l
  bone2Name=clavicle_l
  swing1LimitDegrees=45
  swing1Motion=Free
  swing2LimitDegrees=45
  swing2Motion=Free
  twistLimitDegrees=45
  twistMotion=Free
 }
 {
  bone1Name=upperarm_r
  bone2Name=clavicle_r
  swing1LimitDegrees=45
  swing1Motion=Free
  swing2LimitDegrees=45
  swing2Motion=Free
  twistLimitDegrees=45
  twistMotion=Free
 }
 {
  bone1Name=neck_02
  bone2Name=neck_01
  swing1LimitDegrees=45
  swing1Motion=Free
  swing2LimitDegrees=45
  swing2Motion=Free
  twistLimitDegrees=45
  twistMotion=Free
 }
 {
  bone1Name=head
  bone2Name=neck_02
  swing1LimitDegrees=45
  swing1Motion=Free
  swing2LimitDegrees=45
  swing2Motion=Free
  twistLimitDegrees=45
  twistMotion=Free
 }
]
