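# Control Rig for the Kennedy female character's face accessory skeleton that adjusts hair placement as the head moves.
# Four chained spherical pose readers measure head orientation against neck_01, spine_05 and neck_02, the last two mirroring left and right.
# Each reader's output weights corrective rotations on five front and back hair base bones, which parent the dyn_ chains.
# Two other pose readers and a neck/head control-to-bone copy sit disconnected from execution.
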ControlRigBlueprint /BRCosmetics/Characters/Player/Female/Medium/Bodies/F_MED_Kennedy/Meshes/Parts/F_MED_Kennedy_FaceAcc_CtrlRig
graphs [
 RigVMFunctionLibrary {links [], nodes []}
 RigVMFunction_Sequence_1_ContainedGraph {nodes [
   n1 = RigVMFunctionEntryNode()
   RigVMFunction_Sequence = &RigVMFunction_Sequence(ExecuteContext=@n1.ExecuteContext)
   n2 = &RigVMFunction_Sequence(ExecuteContext=@RigVMFunction_Sequence.B)
   RigVMFunctionReturnNode(A=@RigVMFunction_Sequence.A, B=@n2.A, C=@n2.B)
  ]}
 RigVMModel {nodes [
   RigUnit_BeginExecution = &RigUnit_BeginExecution()
   RigVMFunction_Sequence = RigVMAggregateNode(ExecuteContext=@RigUnit_BeginExecution.ExecutePin)
   n1 = &RigUnit_SphericalPoseReader(
    ActiveRegionScaleFactors=(PositiveWidth=1,NegativeWidth=1,PositiveHeight=1,NegativeHeight=1)
    ActiveRegionSize=0.1
    Debug=(bDrawDebug=false,DebugScale=25,DebugSegments=20,DebugThickness=0.25)
    DriverAxis=(X=0,Y=-1,Z=0)
    DriverItem=(Type=Bone,Name="head")
    FalloffRegionScaleFactors=(PositiveWidth=2,NegativeWidth=2,PositiveHeight=2,NegativeHeight=2)
    FalloffSize=0.2
    FlipWidthScaling=false
    OptionalParentItem=(Type=Bone,Name="neck_01")
    RotationOffset=(X=-90,Y=0,Z=-55)
    ExecutePin=@RigVMFunction_Sequence.A
   )
   > &RigUnit_ModifyTransforms(
    ItemToModify=((Item=(Type=Bone,Name="base_Knd_C_hair_back"),Transform=(Rotation=(X=0,Y=0,Z=-0.6428,W=0.766),Translation=(X=0,Y=0,Z=0),Scale3D=(X=1,Y=1,Z=1))))
    Weight=@n1.OutputParam
   )
   > &RigUnit_ModifyTransforms(
    ItemToModify=((Item=(Type=Bone,Name="base_Knd_L_hair_back"),Transform=(Rotation=(X=0.2346,Y=-0.1094,Z=-0.4082,W=0.8754),Translation=(X=0,Y=0,Z=0),Scale3D=(X=1,Y=1,Z=1))))
    Weight=@n1.OutputParam
   )
   > &RigUnit_ModifyTransforms(
    ItemToModify=((Item=(Type=Bone,Name="base_Knd_R_hair_back"),Transform=(Rotation=(X=-0.2241,Y=0.1294,Z=-0.483,W=0.8365),Translation=(X=0,Y=0,Z=0),Scale3D=(X=1,Y=1,Z=1))))
    Weight=@n1.OutputParam
   )
   > &RigUnit_ModifyTransforms(
    ItemToModify=((Item=(Type=Bone,Name="base_Knd_L_hair_front"),Transform=(Rotation=(X=0,Y=-0.0872,Z=0,W=0.9962),Translation=(X=0,Y=0,Z=0),Scale3D=(X=1,Y=1,Z=1))))
    Weight=@n1.OutputParam
   )
   > &RigUnit_ModifyTransforms(
    ItemToModify=((Item=(Type=Bone,Name="base_Knd_R_hair_front"),Transform=(Rotation=(X=0,Y=0.0872,Z=0,W=0.9962),Translation=(X=0,Y=0,Z=0),Scale3D=(X=1,Y=1,Z=1))))
    Weight=@n1.OutputParam
   )
   > n2 = &RigUnit_SphericalPoseReader(
    ActiveRegionScaleFactors=(PositiveWidth=0.25,NegativeWidth=0.25,PositiveHeight=1,NegativeHeight=1)
    ActiveRegionSize=0.1
    Debug=(bDrawDebug=false,DebugScale=25,DebugSegments=20,DebugThickness=0.25)
    DriverAxis=(X=0,Y=1,Z=0)
    DriverItem=(Type=Bone,Name="head")
    FalloffRegionScaleFactors=(PositiveWidth=0.75,NegativeWidth=0.5,PositiveHeight=4,NegativeHeight=4)
    FalloffSize=0.2
    OptionalParentItem=(Type=Bone,Name="spine_05")
    RotationOffset=(X=90,Y=0,Z=25)
   )
   > &RigUnit_ModifyTransforms(
    ItemToModify=((Item=(Type=Bone,Name="base_Knd_C_hair_back"),Transform=(Rotation=(X=0,Y=0,Z=0,W=1),Translation=(X=0,Y=0,Z=0),Scale3D=(X=1,Y=1,Z=1))))
    Weight=@n2.OutputParam
   )
   > &RigUnit_ModifyTransforms(
    ItemToModify=((Item=(Type=Bone,Name="base_Knd_L_hair_back"),Transform=(Rotation=(X=0,Y=-0.0872,Z=0,W=0.9962),Translation=(X=0,Y=0,Z=0),Scale3D=(X=1,Y=1,Z=1))))
    Weight=@n2.OutputParam
   )
   > &RigUnit_ModifyTransforms(
    ItemToModify=((Item=(Type=Bone,Name="base_Knd_R_hair_back"),Transform=(Rotation=(X=0,Y=0.0872,Z=0,W=0.9962),Translation=(X=0,Y=0,Z=0),Scale3D=(X=1,Y=1,Z=1))))
    Weight=@n2.OutputParam
   )
   > &RigUnit_ModifyTransforms(
    ItemToModify=((Item=(Type=Bone,Name="base_Knd_L_hair_front"),Transform=(Rotation=(X=0.0226,Y=-0.0842,Z=0.2578,W=0.9623),Translation=(X=0,Y=0,Z=0),Scale3D=(X=1,Y=1,Z=1))))
    Weight=@n2.OutputParam
   )
   > &RigUnit_ModifyTransforms(
    ItemToModify=((Item=(Type=Bone,Name="base_Knd_R_hair_front"),Transform=(Rotation=(X=-0.0226,Y=0.0842,Z=0.2578,W=0.9623),Translation=(X=0,Y=0,Z=0),Scale3D=(X=1,Y=1,Z=1))))
    Weight=@n2.OutputParam
   )
   > n3 = &RigUnit_SphericalPoseReader(
    ActiveRegionScaleFactors=(PositiveWidth=1,NegativeWidth=0.5,PositiveHeight=1,NegativeHeight=1)
    ActiveRegionSize=0.1
    Debug=(bDrawDebug=false,DebugScale=25,DebugSegments=20,DebugThickness=0.25)
    DriverAxis=(X=0,Y=-1,Z=0)
    DriverItem=(Type=Bone,Name="head")
    FalloffRegionScaleFactors=(PositiveWidth=1,NegativeWidth=0.5,PositiveHeight=1,NegativeHeight=1)
    FalloffSize=0.2
    OptionalParentItem=(Type=Bone,Name="neck_02")
    RotationOffset=(X=-55,Y=0,Z=-10)
   )
   > &RigUnit_ModifyTransforms(
    ItemToModify=((Item=(Type=Bone,Name="base_Knd_C_hair_back"),Transform=(Rotation=(X=0,Y=0,Z=-0.0872,W=0.9962),Translation=(X=0,Y=0,Z=0),Scale3D=(X=1,Y=1,Z=1))))
    Weight=@n3.OutputParam
   )
   > &RigUnit_ModifyTransforms(
    ItemToModify=((Item=(Type=Bone,Name="base_Knd_L_hair_back"),Transform=(Rotation=(X=0,Y=0,Z=-0.0872,W=0.9962),Translation=(X=0,Y=0,Z=0),Scale3D=(X=1,Y=1,Z=1))))
    Weight=@n3.OutputParam
   )
   > &RigUnit_ModifyTransforms(
    ItemToModify=((Item=(Type=Bone,Name="base_Knd_R_hair_back"),Transform=(Rotation=(X=0.173,Y=-0.0151,Z=-0.0858,W=0.9811),Translation=(X=0,Y=0,Z=0),Scale3D=(X=1,Y=1,Z=1))))
    Weight=@n3.OutputParam
   )
   > &RigUnit_ModifyTransforms(
    ItemToModify=((Item=(Type=Bone,Name="base_Knd_L_hair_front"),Transform=(Rotation=(X=0,Y=-0.1736,Z=0,W=0.9848),Translation=(X=0,Y=0,Z=0),Scale3D=(X=1,Y=1,Z=1))))
    Weight=@n3.OutputParam
   )
   > &RigUnit_ModifyTransforms(
    ItemToModify=((Item=(Type=Bone,Name="base_Knd_R_hair_front"),Transform=(Rotation=(X=0,Y=0,Z=0,W=1),Translation=(X=0,Y=0,Z=0),Scale3D=(X=1,Y=1,Z=1))))
    Weight=@n3.OutputParam
   )
   > n4 = &RigUnit_SphericalPoseReader(
    ActiveRegionScaleFactors=(PositiveWidth=1,NegativeWidth=0.5,PositiveHeight=1,NegativeHeight=1)
    ActiveRegionSize=0.1
    Debug=(bDrawDebug=false,DebugScale=25,DebugSegments=20,DebugThickness=0.25)
    DriverAxis=(X=0,Y=-1,Z=0)
    DriverItem=(Type=Bone,Name="head")
    FalloffRegionScaleFactors=(PositiveWidth=1,NegativeWidth=0.5,PositiveHeight=1,NegativeHeight=1)
    FalloffSize=0.2
    OptionalParentItem=(Type=Bone,Name="neck_02")
    RotationOffset=(X=235,Y=0,Z=-10)
   )
   > &RigUnit_ModifyTransforms(
    ItemToModify=((Item=(Type=Bone,Name="base_Knd_C_hair_back"),Transform=(Rotation=(X=0,Y=0,Z=-0.0872,W=0.9962),Translation=(X=0,Y=0,Z=0),Scale3D=(X=1,Y=1,Z=1))))
    Weight=@n4.OutputParam
   )
   > &RigUnit_ModifyTransforms(
    ItemToModify=((Item=(Type=Bone,Name="base_Knd_L_hair_back"),Transform=(Rotation=(X=0.173,Y=-0.0151,Z=-0.0858,W=0.9811),Translation=(X=0,Y=0,Z=0),Scale3D=(X=1,Y=1,Z=1))))
    Weight=@n4.OutputParam
   )
   > &RigUnit_ModifyTransforms(
    ItemToModify=((Item=(Type=Bone,Name="base_Knd_R_hair_back"),Transform=(Rotation=(X=0,Y=0,Z=-0.0872,W=0.9962),Translation=(X=0,Y=0,Z=0),Scale3D=(X=1,Y=1,Z=1))))
    Weight=@n4.OutputParam
   )
   > &RigUnit_ModifyTransforms(
    ItemToModify=((Item=(Type=Bone,Name="base_Knd_L_hair_front"),Transform=(Rotation=(X=0,Y=0,Z=0,W=1),Translation=(X=0,Y=0,Z=0),Scale3D=(X=1,Y=1,Z=1))))
    Weight=@n4.OutputParam
   )
   > n5 = &RigUnit_ModifyTransforms(
    ItemToModify=((Item=(Type=Bone,Name="base_Knd_R_hair_front"),Transform=(Rotation=(X=0,Y=0.1736,Z=0,W=0.9848),Translation=(X=0,Y=0,Z=0),Scale3D=(X=1,Y=1,Z=1))))
    Weight=@n4.OutputParam
   )
   RigVMRerouteNode(Value=@n5.ExecutePin)
   RigUnit_GetTransform = &RigUnit_GetTransform(Item=(Type=Control,Name="neck_01_ctrl"), bInitial=False)
   &RigUnit_SetTransform(ExecutePin=(), Item=(Type=Bone,Name="neck_01"), Weight=1, bInitial=False, Value=@RigUnit_GetTransform.Transform)
   > &RigUnit_SetTransform(
    Item=(Type=Bone,Name="neck_02")
    Weight=1
    bInitial=False
    Value=&RigUnit_GetTransform(Item=(Type=Control,Name="neck_02_ctrl"), bInitial=False).Transform
   )
   > RigUnit_SetTransform = &RigUnit_SetTransform(
    Item=(Type=Bone,Name="head")
    Weight=1
    bInitial=False
    Value=&RigUnit_GetTransform(Item=(Type=Control,Name="head_ctrl"), bInitial=False).Transform
   )
   n6 = &RigUnit_SphericalPoseReader(
    ActiveRegionScaleFactors=(PositiveWidth=1,NegativeWidth=1,PositiveHeight=1.9,NegativeHeight=1.9)
    ActiveRegionSize=0.1
    Debug=(bDrawDebug=false,bDraw2D=False,bDrawLocalAxes=false,DebugScale=25,DebugSegments=20,DebugThickness=0.25)
    DriverAxis=(X=1,Y=-0.1,Z=0)
    DriverItem=(Type=Bone,Name="head")
    ExecutePin=()
    FalloffRegionScaleFactors=(PositiveWidth=1.9,NegativeWidth=1,PositiveHeight=2,NegativeHeight=2)
    FalloffSize=0.2
    FlipHeightScaling=False
    FlipWidthScaling=False
    OptionalParentItem=(Type=Bone,Name="None")
    RotationOffset=(X=-90,Y=0,Z=30)
   )
   > &RigUnit_ModifyTransforms(
    ItemToModify="((Item=(Type=Bone,Name=\"base_Knd_C_hair_back\"),Transform=(Rotation=(X=0,Y=0,Z=-0.6428,W=0.766),Translation=(X=0,Y=0,Z=0),Scale3D=(X=1,Y=1,Z=1))),(Item=(Type=Bone,Name=\"base_Knd_L_hair_back\"),Transform=(Rotation=(X=0.2346,Y=-0.1094,Z=-0.4082,W=0.8754),Trans ...(truncated)"
    WeightMinimum=0
    Weight=@n6.OutputParam
   )
   n7 = &RigUnit_SphericalPoseReader(
    ActiveRegionScaleFactors=(PositiveWidth=1,NegativeWidth=1,PositiveHeight=1.5,NegativeHeight=1.5)
    ActiveRegionSize=0.1
    Debug=(bDrawDebug=true,bDraw2D=false,bDrawLocalAxes=false,DebugScale=25,DebugSegments=20,DebugThickness=0.25)
    DriverAxis=(X=1,Y=-0.1,Z=0)
    DriverItem=(Type=Bone,Name="head")
    ExecutePin=()
    FalloffRegionScaleFactors=(PositiveWidth=2.8,NegativeWidth=1,PositiveHeight=2,NegativeHeight=2)
    FalloffSize=0.2
    FlipHeightScaling=False
    FlipWidthScaling=False
    OptionalParentItem=(Type=Bone,Name="None")
    RotationOffset=(X=-32.955,Y=-13.7527,Z=24.9838)
   )
   > &RigUnit_ModifyTransforms(
    ItemToModify="((Item=(Type=Bone,Name=\"base_Knd_C_hair_back\"),Transform=(Rotation=(X=0.0436,Y=0.4981,Z=-0.0755,W=0.8627),Translation=(X=0,Y=0,Z=0),Scale3D=(X=1,Y=1,Z=1))),(Item=(Type=Bone,Name=\"base_Knd_L_hair_back\"),Transform=(Rotation=(X=0,Y=0.5,Z=0,W=0.866),Translatio ...(truncated)"
    WeightMinimum=0
    Weight=@n7.OutputParam
   )
  ]}
]
hierarchy [
 root {loc=<0 0 0>, rot=<0 0 0 1>}
 pelvis {parent=root, loc=<0 -0.5301 98.0571>, rot=<0.0127 0.707 -0.0127 -0.707>}
 spine_01 {parent=pelvis, loc=<6.1208 -0.9328 0>, rot=<0 0 0.034 -0.9994>}
 spine_02 {parent=spine_01, loc=<6.5996 -0.2095 0>, rot=<0 0 0.0238 -0.9997>}
 spine_03 {parent=spine_02, loc=<6.4914 -0.1237 0>, rot=<0 0 0.0149 -0.9999>}
 spine_04 {parent=spine_03, loc=<6.9598 0.0794 0>, rot=<0 0 -0.0132 -0.9999>}
 spine_05 {parent=spine_04, loc=<6.9987 0 0>, rot=<0 0 -0.0015 -1>}
 neck_01 {parent=spine_05, loc=<11.3138 0.1529 0>, rot=<0 0 -0.2003 -0.9797>}
 neck_02 {parent=neck_01, loc=<4.5627 0 0>, rot=<0 0 0.0504 -0.9987>}
 head {parent=neck_02, loc=<5.3914 0 0>, rot=<0 0 0.115 -0.9934>}
 base_Knd_cap {parent=head, loc=<3.5202 0.2424 0.0256>, rot=<0.0029 -0.7071 -0.0029 -0.7071>}
 dyn_Knd_cap {parent=base_Knd_cap, loc=<0 0 0>, rot=<0 0 0 1>}
 base_Knd_L_hair_front {parent=head, loc=<3.3315 2.863 -6.5177>, rot=<-0.0688 -0.9965 0.0018 -0.0468>}
 dyn_Knd_L_hair_front_1 {parent=base_Knd_L_hair_front, loc=<0 0 0>, rot=<0 0 0 1>}
 dyn_Knd_L_hair_front_2 {parent=dyn_Knd_L_hair_front_1, loc=<4.5479 0 0>, rot=<0 0 0 1>}
 dyn_Knd_L_hair_front_3 {parent=dyn_Knd_L_hair_front_2, loc=<4.5479 0 0>, rot=<0 0 0 1>}
 dyn_Knd_L_hair_front_4 {parent=dyn_Knd_L_hair_front_3, loc=<4.5479 0 0>, rot=<0 0 0 1>}
 base_Knd_C_hair_back {parent=head, loc=<3.8544 -5.4819 0>, rot=<0.1897 -0.9819 0 0>}
 dyn_Knd_C_hair_back_1 {parent=base_Knd_C_hair_back, loc=<0 0 0>, rot=<0 0 0 1>}
 dyn_Knd_C_hair_back_2 {parent=dyn_Knd_C_hair_back_1, loc=<4.663 0 0>, rot=<0 0 0 1>}
 dyn_Knd_C_hair_back_3 {parent=dyn_Knd_C_hair_back_2, loc=<4.663 0 0>, rot=<0 0 0 1>}
 dyn_Knd_C_hair_back_4 {parent=dyn_Knd_C_hair_back_3, loc=<4.663 0 0>, rot=<0 0 0 1>}
 base_Knd_R_hair_back {parent=head, loc=<3.89 -0.925 6.484>, rot=<0.2724 -0.9622 0 0>}
 dyn_Knd_R_hair_back_1 {parent=base_Knd_R_hair_back, loc=<0 0 0>, rot=<0 0 0 1>}
 dyn_Knd_R_hair_back_2 {parent=dyn_Knd_R_hair_back_1, loc=<4.663 0 0>, rot=<0 0 0 1>}
 dyn_Knd_R_hair_back_3 {parent=dyn_Knd_R_hair_back_2, loc=<4.663 0 0>, rot=<0 0 0 1>}
 dyn_Knd_R_hair_back_4 {parent=dyn_Knd_R_hair_back_3, loc=<4.663 0 0>, rot=<0 0 0 1>}
 dyn_Knd_R_hair_back_1_a {parent=dyn_Knd_R_hair_back_1, loc=<4.2347 1.2273 -3.6369>, rot=<-0.4503 -0.2165 -0.0826 -0.8623>}
 dyn_Knd_R_hair_back_3_a {parent=dyn_Knd_R_hair_back_1_a, loc=<5.7037 0 0>, rot=<0.491 0.0926 -0.1434 -0.8543>}
 base_Knd_L_hair_back {parent=head, loc=<3.8897 -0.9251 -6.484>, rot=<0.2724 -0.9622 0 0>}
 dyn_Knd_L_hair_back_1 {parent=base_Knd_L_hair_back, loc=<0 0 0>, rot=<0 0 0 1>}
 dyn_Knd_L_hair_back_2 {parent=dyn_Knd_L_hair_back_1, loc=<4.663 0 0>, rot=<0 0 0 1>}
 dyn_Knd_L_hair_back_3 {parent=dyn_Knd_L_hair_back_2, loc=<4.663 0 0>, rot=<0 0 0 1>}
 dyn_Knd_L_hair_back_4 {parent=dyn_Knd_L_hair_back_3, loc=<4.663 0 0>, rot=<0 0 0 1>}
 dyn_Knd_L_hair_back_3_a {parent=dyn_Knd_L_hair_back_3, loc=<0 3.152 5.342>, rot=<0 0 -0.2588 -0.9659>}
 base_Knd_R_hair_front {parent=head, loc=<3.332 2.863 6.518>, rot=<-0.0686 -0.9957 -0.0057 0.0618>}
 dyn_Knd_R_hair_front_1 {parent=base_Knd_R_hair_front, loc=<0 0 0>, rot=<0 0 0 1>}
 dyn_Knd_R_hair_front_2 {parent=dyn_Knd_R_hair_front_1, loc=<4.5479 0 0>, rot=<0 0 0 1>}
 dyn_Knd_R_hair_front_3 {parent=dyn_Knd_R_hair_front_2, loc=<4.5479 0 0>, rot=<0 0 0 1>}
 dyn_Knd_R_hair_front_4 {parent=dyn_Knd_R_hair_front_3, loc=<4.5479 0 0>, rot=<0 0 0 1>}
 clavicle_l {parent=spine_05, loc=<5.4831 -0.1053 -3.2946>, rot=<0.0465 -0.7289 -0.009 -0.6829>}
 upperarm_l {parent=clavicle_l, loc=<9.8658 -1.2136 -0.4105>, rot=<0.0164 -0.3245 -0.0269 -0.9454>}
 clavicle_r {parent=spine_05, loc=<5.4831 -0.1053 3.2946>, rot=<0.7289 0.0465 -0.6829 0.009>}
 upperarm_r {parent=clavicle_r, loc=<-9.8658 1.2136 0.4105>, rot=<0.0164 -0.3245 -0.0269 -0.9454>}
 neck_01_ctrl {controlType=EULER_TRANSFORM, type=Control, loc=<0 0 0>, rot=<0 0 0 1>}
 neck_02_ctrl {controlType=EULER_TRANSFORM, parent=neck_01_ctrl, type=Control, loc=<0 0 0>, rot=<0 0 0 1>}
 head_ctrl {controlType=EULER_TRANSFORM, parent=neck_02_ctrl, type=Control, loc=<0 0 0>, rot=<0 0 0 1>}
]
