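# Physics asset for the FrontEnd version of the Kennedy character's front hair, defining capsule bodies and joints for left and right strands.
# Each strand is a four-bone dyn chain from a base bone on the head, with tight swing, near-zero twist and capsules widening toward the tips.
# Head, neck, clavicle and upper-arm capsules approximate the surrounding body, while their joints and the hair-base joints leave rotation free.
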PhysicsAsset /BRCosmetics/Characters/Player/Female/Medium/Bodies/F_MED_Kennedy/Meshes/Parts/F_MED_Kennedy_FaceAcc_FrontHair_FrontEnd_Physics
bodies {
 base_Knd_L_hair_front [{shapeType=Capsule, length=0.505, radius=0.505, shapeName=base_Knd_L_hair_front_capsule, rotation=<rot 0 -90 0>}]
 base_Knd_R_hair_front [{shapeType=Capsule, length=0.505, radius=0.505, shapeName=base_Knd_R_hair_front_capsule, rotation=<rot 0 -90 0>}]
 clavicle_l [{shapeType=Capsule, length=16, radius=6, shapeName=clavicle_l_capsule, center=<0 -1.2 -1.5777>, rotation=<rot 90 0 0>}]
 clavicle_r [{shapeType=Capsule, length=16, radius=6, shapeName=clavicle_r_capsule, center=<0 1.2 1.5>, rotation=<rot 90 0 0>}]
 dyn_Knd_L_hair_front_1 [{shapeType=Capsule, length=4.1598, radius=0.8402, shapeName=dyn_Knd_L_hair_front_1_capsule, center=<2 0 0>, rotation=<rot 90 0 0>}]
 dyn_Knd_L_hair_front_2 [{shapeType=Capsule, length=3.6079, radius=1.3921, shapeName=dyn_Knd_L_hair_front_2_capsule, center=<2 0 0>, rotation=<rot 90 0 0>}]
 dyn_Knd_L_hair_front_3 [{
   shapeType=Capsule
   length=0.9674
   radius=4.1973
   shapeName=dyn_Knd_L_hair_front_3_capsule
   center=<2 -0.9904 0>
   rotation=<rot 90 -91.7836 -71.7837>
  }]
 dyn_Knd_L_hair_front_4 [{shapeType=Capsule, length=3.7755, radius=4, shapeName=dyn_Knd_L_hair_front_4_capsule, center=<2 -0.3 1>, rotation=<rot 0 90 90>}]
 dyn_Knd_R_hair_front_1 [{shapeType=Capsule, length=3.8246, radius=1.1754, shapeName=dyn_Knd_R_hair_front_1_capsule, center=<2 0 -0.7>, rotation=<rot 90 0 0>}]
 dyn_Knd_R_hair_front_2 [{shapeType=Capsule, length=3.3182, radius=2, shapeName=dyn_Knd_R_hair_front_2_capsule, center=<1.7616 0 -0.3>, rotation=<rot 90 0 0>}]
 dyn_Knd_R_hair_front_3 [{shapeType=Capsule, length=4, radius=4, shapeName=dyn_Knd_R_hair_front_3_capsule, center=<2.9406 0 -0.5>, rotation=<rot 0 0.0002 0>}]
 dyn_Knd_R_hair_front_4 [{
   shapeType=Capsule
   length=5
   radius=4.6
   shapeName=dyn_Knd_R_hair_front_4_capsule
   center=<3.3116 1.7616 -1.0391>
   rotation=<rot 0 0.0004 15>
  }]
 head [
  {shapeType=Capsule, length=4.9246, radius=5.5636, shapeName=head_capsule, center=<0.9738 4.2592 0>, rotation=<rot 0 -90.0003 -90>}
  {shapeType=Capsule, length=11.1744, radius=0.8963, shapeName=head_capsule, center=<3.3994 9.648 0>}
 ]
 neck_01 [{shapeType=Capsule, length=4, radius=4, shapeName=neck_01_capsule, rotation=<rot 90 0 0>}]
 upperarm_l [{shapeType=Capsule, length=26.1871, radius=4.1, shapeName=upperarm_l_capsule, center=<13.0855 -1.4918 -1.2183>, rotation=<rot 90 0 0>}]
 upperarm_r [{shapeType=Capsule, length=11, radius=3.5, shapeName=upperarm_l_capsule, center=<-8 0.3 1.2>, rotation=<rot 90 0 180>}]
}
constraints [
 {
  bone1Name=head
  bone2Name=neck_01
  swing1LimitDegrees=45
  swing1Motion=Free
  swing2LimitDegrees=45
  swing2Motion=Free
  twistLimitDegrees=45
  twistMotion=Free
 }
 {
  bone1Name=upperarm_r
  bone2Name=clavicle_r
  swing1LimitDegrees=45
  swing1Motion=Free
  swing2LimitDegrees=45
  swing2Motion=Free
  twistLimitDegrees=45
  twistMotion=Free
 }
 {
  bone1Name=base_Knd_L_hair_front
  bone2Name=head
  swing1LimitDegrees=45
  swing1Motion=Free
  swing2LimitDegrees=45
  swing2Motion=Free
  twistLimitDegrees=45
  twistMotion=Free
 }
 {bone1Name=dyn_Knd_L_hair_front_1, bone2Name=base_Knd_L_hair_front, swing1LimitDegrees=20, swing2LimitDegrees=16, twistLimitDegrees=1}
 {bone1Name=dyn_Knd_L_hair_front_2, bone2Name=dyn_Knd_L_hair_front_1, swing1LimitDegrees=8, swing2LimitDegrees=16, twistLimitDegrees=1}
 {bone1Name=dyn_Knd_L_hair_front_3, bone2Name=dyn_Knd_L_hair_front_2, swing1LimitDegrees=16, swing2LimitDegrees=16, twistLimitDegrees=1}
 {bone1Name=dyn_Knd_L_hair_front_4, bone2Name=dyn_Knd_L_hair_front_3, swing1LimitDegrees=16, swing2LimitDegrees=16, twistLimitDegrees=1}
 {
  bone1Name=base_Knd_R_hair_front
  bone2Name=head
  swing1LimitDegrees=45
  swing1Motion=Free
  swing2LimitDegrees=45
  swing2Motion=Free
  twistLimitDegrees=45
  twistMotion=Free
 }
 {bone1Name=dyn_Knd_R_hair_front_1, bone2Name=base_Knd_R_hair_front, swing1LimitDegrees=16, swing2LimitDegrees=16, twistLimitDegrees=1}
 {bone1Name=dyn_Knd_R_hair_front_2, bone2Name=dyn_Knd_R_hair_front_1, swing1LimitDegrees=8, swing2LimitDegrees=16, twistLimitDegrees=1}
 {bone1Name=dyn_Knd_R_hair_front_3, bone2Name=dyn_Knd_R_hair_front_2, swing1LimitDegrees=16, swing2LimitDegrees=16, twistLimitDegrees=1}
 {bone1Name=dyn_Knd_R_hair_front_4, bone2Name=dyn_Knd_R_hair_front_3, swing1LimitDegrees=16, swing2LimitDegrees=16, twistLimitDegrees=1}
 {
  bone1Name=upperarm_l
  bone2Name=clavicle_l
  swing1LimitDegrees=45
  swing1Motion=Free
  swing2LimitDegrees=45
  swing2Motion=Free
  twistLimitDegrees=45
  twistMotion=Free
 }
]
