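# Post-process style face accessory AnimBP for the female Kennedy character, simulating and correcting front and back hair.
# Copied parent pose, with shoulders layered from the body mesh, feeds a Control Rig, then front and back RigidBody passes in spine_05 space.
# Front-end and tactical-sprint flags select physics variants.
# Additive bone offsets adjust hair for skydive floating and pistol aiming, using spine-to-head pitch alphas from the event graph.
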AnimBlueprint /BRCosmetics/Characters/Player/Female/Medium/Bodies/F_MED_Kennedy/Meshes/Parts/F_MED_Kennedy_FaceAcc_AnimBP
target_skeleton=&/F_MED_Kennedy_FaceAcc_Skeleton
graphs [
 AnimGraph {nodes [
   n1 = RigidBody!(ComponentPose=Local To Component(LocalPose=Use cached pose 'Cache_FrontHair'()), Alpha, OverrideWorldGravity:Vector="0,0,0", ExternalForce) -> (Pose) {
    bClampLinearTranslationLimitToRefPose=true
    bOverrideWorldGravity=true
    overridePhysicsAsset=&/F_MED_Kennedy_FaceAcc_BackHair_Physics
    simulationSpace=BaseBoneSpace
    alphaBoolBlend {blendOption=Linear}
    baseBoneRef {boneName=spine_05}
    componentLinearAccScale=<0.25 0.1 0.1>
    componentLinearVelScale=<0.25 0.1 0.1>
    simSpaceSettings {worldAlpha=0.2}
   }
   n2 = RigidBody(ComponentPose=Local To Component(LocalPose=Use cached pose 'Cache_FrontHair'()), OverrideWorldGravity="0,0,0") {
    bClampLinearTranslationLimitToRefPose=true
    bOverrideWorldGravity=true
    overridePhysicsAsset=&F_MED_Kennedy_FaceAcc_BackHair_Physics
    simulationSpace=BaseBoneSpace
    alphaBoolBlend {blendOption=Linear}
    baseBoneRef {boneName=spine_05}
    componentAppliedLinearAccClamp=<3000 3000 3000>
    componentLinearAccScale=<0.25 0.1 0.1>
    componentLinearVelScale=<0.25 0.1 0.1>
    simSpaceSettings {
     maxAngularAcceleration=1000
     maxAngularVelocity=3
     maxLinearAcceleration=700
     maxLinearVelocity=500
     velocityScaleZ=0
     worldAlpha=0.5
     externalLinearDragV=<0.2 0.2>
    }
   }
   n3 = Blend Poses by bool(
    BlendPose_0=Component To Local(ComponentPose=@n2)
    BlendPose_1=Component To Local(ComponentPose=@n1)
    bActiveValue=Get bIsTacticalSprinting()
   ) {blendType=Linear}
   n4 = RigidBody(ComponentPose=Local To Component(LocalPose=Use cached pose 'Cache_FrontHair'()), OverrideWorldGravity="0,0,0") {
    bClampLinearTranslationLimitToRefPose=true
    bOverrideWorldGravity=true
    overridePhysicsAsset=&/F_MED_Kennedy_FaceAcc_BackHair_FrontEnd_Physics
    simulationSpace=BaseBoneSpace
    alphaBoolBlend {blendOption=Linear}
    baseBoneRef {boneName=spine_05}
    componentAppliedLinearAccClamp=<3000 3000 3000>
    componentLinearAccScale=<0.25 0.1 0.1>
    componentLinearVelScale=<0.25 0.1 0.1>
    simSpaceSettings {
     maxAngularAcceleration=1000
     maxAngularVelocity=3
     maxLinearAcceleration=700
     maxLinearVelocity=500
     velocityScaleZ=0
     worldAlpha=0.5
     externalLinearDragV=<0.2 0.2>
    }
   }
   n5 = "Transform (Modify) Bone - Bone: base_Knd_L_hair_back"(ComponentPose=@n4, Alpha=1, Translation="-0.2,-0.2,0", Rotation="0,0,0", Scale="1,1,1") {
    rotationMode=BMM_Additive
    rotationSpace=BCS_BoneSpace
    translationMode=BMM_Additive
    translationSpace=BCS_BoneSpace
    alphaBoolBlend {blendOption=Linear}
    alphaScaleBiasClamp {bClampResult=true, bInterpResult=true}
    boneToModify {boneName=base_Knd_L_hair_back}
    translation=<-0.2 -0.2>
   }
   n6 = Component To Local(ComponentPose=@n5)
   n7 = Get bIsInFrontEnd()
   n8 = Get AimingUp() -> (AimingUp:double)
   n9 = "Transform (Modify) Bone - Bone: base_Knd_L_hair_back"(
    ComponentPose=Local To Component(LocalPose=Blend Poses by bool(BlendPose_0=@n6, BlendPose_1=@n3, bActiveValue=@n7) {blendType=Linear})
    Alpha=@n8
    Translation="0,0,0"
    Rotation="-10,10,-10"
    Scale="1,1,1"
   ) {
    rotationMode=BMM_Additive
    rotationSpace=BCS_BoneSpace
    alphaBoolBlend {blendOption=Linear}
    alphaScaleBiasClamp {bInterpResult=true}
    boneToModify {boneName=base_Knd_L_hair_back}
    rotation=<rot -10 10 -10>
   }
   n10 = "Transform (Modify) Bone - Bone: base_Knd_C_hair_back"(ComponentPose=@n9, Alpha=@n8, Translation="0,0,0", Rotation="0,7,0", Scale="1,1,1") {
    rotationMode=BMM_Additive
    rotationSpace=BCS_BoneSpace
    alphaBoolBlend {blendOption=Linear}
    alphaScaleBiasClamp {bInterpResult=true}
    boneToModify {boneName=base_Knd_C_hair_back}
    rotation {yaw=7}
   }
   n11 = Get Skydive() -> (Skydive:double)
   n12 = "Transform (Modify) Bone - Bone: base_Knd_C_hair_back"(ComponentPose=@n10, Alpha=@n11, Translation="0,0,0", Rotation="0,15,0", Scale="1,1,1") {
    rotationMode=BMM_Additive
    rotationSpace=BCS_BoneSpace
    alphaBoolBlend {blendOption=Linear}
    alphaScaleBiasClamp {bInterpResult=true}
    boneToModify {boneName=base_Knd_C_hair_back}
    rotation {yaw=15}
   }
   n13 = "Transform (Modify) Bone - Bone: base_Knd_R_hair_back"(ComponentPose=@n12, Alpha=@n11, Translation="0,0,0", Rotation="-10,5,0", Scale="1,1,1") {
    rotationMode=BMM_Additive
    rotationSpace=BCS_BoneSpace
    alphaBoolBlend {blendOption=Linear}
    alphaScaleBiasClamp {bInterpResult=true}
    boneToModify {boneName=base_Knd_R_hair_back}
    rotation {pitch=-10, yaw=5}
   }
   Output Pose(Result=Component To Local(ComponentPose=@n13)) {layerGroup=None}
   n14 = RigidBody(ComponentPose=Local To Component(LocalPose=Use cached pose 'Cache_ControlRig'()), OverrideWorldGravity="0,0,0") {
    bClampLinearTranslationLimitToRefPose=true
    bOverrideWorldGravity=true
    overridePhysicsAsset=&/F_MED_Kennedy_FaceAcc_FrontHair_FrontEnd_Physics
    simulationSpace=BaseBoneSpace
    alphaBoolBlend {blendOption=Linear}
    baseBoneRef {boneName=spine_05}
    componentAppliedLinearAccClamp=<3000 3000 3000>
    componentLinearAccScale=<0.25 0.1 0.1>
    componentLinearVelScale=<0.25 0.1 0.1>
    simSpaceSettings {
     maxAngularAcceleration=1000
     maxAngularVelocity=3
     maxLinearAcceleration=700
     maxLinearVelocity=500
     velocityScaleZ=0
     worldAlpha=0.5
     externalLinearDragV=<0.2 0.2>
    }
   }
   n15 = Component To Local(ComponentPose=@n14)
   n16 = RigidBody(ComponentPose=Local To Component(LocalPose=Use cached pose 'Cache_ControlRig'()), OverrideWorldGravity="0,0,0") {
    bClampLinearTranslationLimitToRefPose=true
    bOverrideWorldGravity=true
    overridePhysicsAsset=&/F_MED_Kennedy_FaceAcc_FrontHair_Physics
    simulationSpace=BaseBoneSpace
    alphaBoolBlend {blendOption=Linear}
    baseBoneRef {boneName=spine_05}
    componentLinearAccScale=<0.2 0.1 0.1>
    componentLinearVelScale=<0.2 0.1 0.1>
    simSpaceSettings {worldAlpha=0.2}
   }
   n17 = Component To Local(ComponentPose=@n16)
   n18 = Get bIsInFrontEnd()
   n19 = "Transform (Modify) Bone - Bone: base_Knd_R_hair_front"(
    ComponentPose=Local To Component(LocalPose=Blend Poses by bool(BlendPose_0=@n15, BlendPose_1=@n17, bActiveValue=@n18) {blendType=Linear})
    Alpha=Get AimingUp()
    Translation="0,0,0"
    Rotation="0,20,10"
    Scale="1,1,1"
   ) {
    rotationMode=BMM_Additive
    rotationSpace=BCS_BoneSpace
    alphaBoolBlend {blendOption=Linear}
    alphaScaleBiasClamp {bClampResult=true, bInterpResult=true}
    boneToModify {boneName=base_Knd_R_hair_front}
    rotation {roll=10, yaw=20}
   }
   n20 = "Transform (Modify) Bone - Bone: base_Knd_R_hair_front"(ComponentPose=@n19, Alpha=Get Targeting() -> (Targeting:double), Translation="0,0,0", Rotation="-5,0,0", Scale="1,1,1") {
    rotationMode=BMM_Additive
    rotationSpace=BCS_BoneSpace
    translationMode=BMM_Additive
    translationSpace=BCS_BoneSpace
    alphaBoolBlend {blendOption=Linear}
    alphaScaleBiasClamp {bClampResult=true, bInterpResult=true}
    boneToModify {boneName=base_Knd_R_hair_front}
    rotation {pitch=-5}
   }
   n21 = "Transform (Modify) Bone - Bone: base_Knd_R_hair_front"(ComponentPose=@n20, Alpha=Get Aiming Down() -> (Aiming Down:double), Translation="0,0,2", Rotation="0,-3,5", Scale="1,1,1") {
    rotationMode=BMM_Additive
    rotationSpace=BCS_BoneSpace
    translationMode=BMM_Additive
    translationSpace=BCS_BoneSpace
    alphaBoolBlend {blendOption=Linear}
    alphaScaleBiasClamp {bClampResult=true, bInterpResult=true}
    boneToModify {boneName=base_Knd_R_hair_front}
    rotation {roll=5, yaw=-3}
    translation {z=2}
   }
   Save cached pose 'Cache_FrontHair'(Pose=Component To Local(ComponentPose=@n21)) {cacheName=Cache_FrontHair}
   n22 = Control Rig(Source=Use cached pose 'Cache_StartPose'()) {
    alphaBoolBlend {blendInTime=1, blendOption=Linear, blendOutTime=1}
    controlRigAssetReference {blueprintRigClass=&/F_MED_Kennedy_FaceAcc_CtrlRig.F_MED_Kennedy_FaceAcc_CtrlRig_C}
   }
   Save cached pose 'Cache_ControlRig'(Pose=@n22) {cacheName=Cache_ControlRig}
   n23 = Layered blend per bone(BasePose=Copy Pose From Mesh() {bUseAttachedParent=true}, BlendPoses_0=Copy Pose From Mesh(SourceMeshComponent=Get Body_Mesh())) {layerSetup [{branchFilters [
       {boneName=clavicle_l}
       {boneName=clavicle_r}
       {boneName=upperarm_l}
       {boneName=upperarm_r}
       {boneName=upperarm_twist_01_l}
       {boneName=upperarm_twist_01_r}
       {boneName=deform_upperarm_bk_base_l}
       {boneName=deform_upperarm_bk_l}
       {boneName=deform_upperarm_fr_base_l}
       {boneName=deform_upperarm_fr_l}
       {boneName=deform_upperarm_up_base_l}
       {boneName=deform_upperarm_up_l}
       {boneName=deform_upperarm_bk_base_r}
       {boneName=deform_upperarm_bk_r}
       {boneName=deform_upperarm_fr_base_r}
       {boneName=deform_upperarm_fr_r}
       {boneName=deform_upperarm_up_base_r}
       {boneName=deform_upperarm_up_r}
      ]}]}
   Save cached pose 'Cache_StartPose'(Pose=@n23) {cacheName=Cache_StartPose}
  ]}
 EventGraph {nodes [
   Event BlueprintUpdateAnimation()
   > n1 = Sequence!(execute) -> (then_0, then_1, then_2:Exec, then_3:Exec, then_4:Exec, then_5:Exec, then_6:Exec, then_7:Exec, then_8:Exec)
   Set Skydive(execute=@n1.then_3, Skydive:double="To Float (Boolean)"(InBool=Get bIsSkydiveFloating())) -> (Output_Get:double)
   n2 = Cast To FortPlayerAnimInstance(execute=@n1.then_8, Object=GetAnimInstance(self=GetOwnerSkeletalMeshComponent()))
   > n3 = Switch on EFortWeaponCoreAnimation!(execute, Selection:EFortWeaponCoreAnimation Enum=Get WeaponCoreAnim(self=@n2)) -> (
    Melee:Exec
    Pistol:Exec
    Shotgun:Exec
    PaperBlueprint:Exec
    Rifle:Exec
    MeleeOneHand:Exec
    MachinePistol:Exec
    RocketLauncher:Exec
    GrenadeLauncher:Exec
    GoingCommando:Exec
    AssaultRifle:Exec
    TacticalShotgun:Exec
    SniperRifle:Exec
    TrapPlacement:Exec
    ShoulderLauncher:Exec
    AbilityDecoTool:Exec
    Crossbow:Exec
    C4:Exec
    RemoteControl:Exec
    DualWield:Exec
    AR_BullPup:Exec
    AR_ForwardGrip:Exec
    MedPackPaddles:Exec
    SMG_P90:Exec
    AR_DrumGun:Exec
    Consumable_Small:Exec
    Consumable_Large:Exec
    Balloon:Exec
    MountedTurret:Exec
    CreativeTool:Exec
    ExplosiveBow:Exec
    AshtonIndigo:Exec
    AshtonChicago:Exec
    MeleeDualWield:Exec
    Unarmed:Exec
   )
   n4 = Branch(execute=@n3.Pistol, Condition=Get bIsTargeting())
   > Set Pistol Equipped(Pistol Equipped:bool=true) -> (Output_Get:bool)
   Set Pistol Equipped(execute=@n4.else, Pistol Equipped=false)
   n5 = Gravity Override State Function DEPRECATED(
    execute=@n1.then_0
    CharacterPartAnimInstance=Self-Reference()
    GameTimeInSeconds=GetGameTimeInSeconds()
    GravityOverrideParams=Get Gravity Override Params()
   )
   > Set GravityOverride(GravityOverride=@n5)
   Set Targeting(execute=@n1.then_4, Targeting:double="To Float (Boolean)"(InBool=Get Pistol Equipped() -> (Pistol Equipped:bool))) -> (Output_Get:double)
   n6 = "To Float (Boolean)"(InBool=Get Pistol Equipped())
   n7 = float * float(
    A=GetAlphaFromReferenceTransformDelta(SkeletalMeshComponent=Get Skel Mesh(), BoneName=spine_05, BaseName=head, InputRange=(X=0,Y=90), TwistAxis="0,0,1")
    B=3.3
   ) {is=title K2Node_PromotableOperator}
   Set AimingUp(execute=@n1.then_1, AimingUp:double="Clamp (Float)"(Value=float * float(A=@n7, B=@n6), Min=0, Max=1)) -> (Output_Get:double)
   n8 = "To Float (Boolean)"(InBool=Get Pistol Equipped())
   n9 = float * float(
    A=GetAlphaFromReferenceTransformDelta(SkeletalMeshComponent=Get Skel Mesh(), BoneName=spine_05, BaseName=head, InputRange=(X=-90,Y=0), OutputRange=(X=-1,Y=0), TwistAxis="0,0,1")
    B=-15
   )
   Set Aiming Down(execute=@n1.then_2, Aiming Down:double="Clamp (Float)"(Value=float * float(A=@n9, B=@n8), Min=0, Max=1)) -> (Output_Get:double)
   Set HairAlpha(execute=@n1.then_6, HairAlpha:double=float * float(A=1, B=Get Hair Skydiving Alpha() -> (Hair Skydiving Alpha:double))) -> (Output_Get:double)
   n10 = Branch(execute=@n1.then_5, Condition=Get bIsSkyDiving())
   > Set Hair Skydiving Alpha(Hair Skydiving Alpha:double=0.4) -> (Output_Get:double)
   Set Hair Skydiving Alpha(execute=@n10.else, Hair Skydiving Alpha=1)
   Event OnCharacterCustomizationLoaded()
   > n11 = Cast To PlayerPawn_Athena(Object=TryGetPawnOwner())
   > Set Body_Mesh(Body_Mesh=GetSkeletalMeshForPartType(self=@n11))
   Event BlueprintInitializeAnimation()
   > Set Skel Mesh(Skel Mesh=GetOwnerSkeletalMeshComponent())
  ]}
]
variables [
 Aiming Down {type=number}
 AimingUp {type=number}
 BackHead_Component_Acc {type=Vector, default=<0.25 0.1 0.1>}
 Body_Mesh {type=SkeletalMeshComponent}
 Gravity Override Params {type=GravityOverrideParamsStruct, default {skydive_FloatingDown {gravityOverride=<0 -700 901>}}}
 GravityOverride {type=Vector}
 Hair Skydiving Alpha {type=number}
 HairAlpha {type=number}
 InFrontEnd {type=number}
 Pistol Equipped {type=boolean}
 Running {type=number}
 Skel Mesh {type=SkeletalMeshComponent}
 Skydive {type=number}
 Targeting {type=number}
]
